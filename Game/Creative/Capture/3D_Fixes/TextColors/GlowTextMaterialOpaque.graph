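# Glowing text material for 3D text in the Creative Capture content, set to masked, two-sided Default Lit despite the Opaque name.
# A Roboto distance-field font parameter's red channel drives a steep lerp between large negative and positive values, which yields hard glyph edges when used as a mask.
# Vertex color is multiplied by a constant boost, so the text color comes from the mesh and is brightened for the glow.
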Material /Game/Creative/Capture/3D_Fixes/TextColors/GlowTextMaterialOpaque
graph {
 properties {blendMode=BLEND_Masked, pixelDepthOffsetMode=PDOM_Legacy, twoSided=true, usedShadingModels=Default Lit}
 nodes [
  LinearInterpolate(Alpha=FontSampleParameter() {font=/Engine/EngineFonts/RobotoDistanceField, parameterName=Font}.R) {constA=-90, constB=90}
  Multiply(A=VertexColor(), B=Constant() {r=5})
 ]
}
